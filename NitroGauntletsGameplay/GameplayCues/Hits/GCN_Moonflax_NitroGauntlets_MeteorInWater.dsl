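; Burst gameplay cue for the Nitro Gauntlets that plays a water explosion effect, named for a meteor hit in water.
; On burst, it line-traces vertically from above to below the target actor's location against the object types held in a variable.
; It spawns the auto-released, pooled Niagara water explosion at the trace's hit location, without checking that the trace hit or reading the cue parameters.
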(fn OnBurst(EitherVFX-preferthisversion)
  (params (MyTarget :type Actor%20Object%20Reference) (Parameters :type Gameplay%20Cue%20Parameters%20Structure%20%28by%20ref%29) (ParticleComponents :type Array%20of%20FXSystem%20Component%20Object%20References) (AudioComponents :type Array%20of%20Audio%20Component%20Object%20References) (BurstCameraShakeInstance :type Camera%20Shake%20Base%20Object%20Reference) (BurstDecalInstance :type Decal%20Actor%20Object%20Reference))
  (bind _returnvalue_location (Math|Transform|BreakTransform (Transformation|GetActorTransform MyTarget)))
  (bind _outhit (Collision|LineTraceForObjects 0 (+ _returnvalue_location "0.000000,0.000000,500.000000") (- _returnvalue_location "0.000000,0.000000,500.000000") (Variables|Default|GetObject%20Types)))
  (Niagara|SpawnSystematLocation :SystemTemplate "/NitroGauntletsGameplay/Assets/FX/NS_NitroGauntlets_WaterExplosion.NS_NitroGauntlets_WaterExplosion" :Location (.Location (Collision|BreakHitResult _outhit)) :Scale "1.000000,1.000000,1.000000" :bAutoDestroy true :bAutoActivate true :PoolingMethod "AutoRelease" :bPreCullCheck true)
  (return))
