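; Gameplay cue burst for the Nitro Gauntlets meteor landing, playing cosmetic impact effects at the target actor.
; A downward trace spawns a surface Niagara effect, switched to sand, snow or grass by physical material lookup.
; On a trace hit with physics fields enabled, it also spawns a radial force field and sets LocalView for locally viewed pawns.
; Separately, it shows a gunshot-type sound indicator to neutral and hostile players.
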(fn OnBurst(EitherVFX-preferthisversion)
  (params (MyTarget :type Actor%20Object%20Reference) (Parameters :type Gameplay%20Cue%20Parameters%20Structure%20%28by%20ref%29) (ParticleComponents :type Array%20of%20FXSystem%20Component%20Object%20References) (AudioComponents :type Array%20of%20Audio%20Component%20Object%20References) (BurstCameraShakeInstance :type Camera%20Shake%20Base%20Object%20Reference) (BurstDecalInstance :type Decal%20Actor%20Object%20Reference))
  (bind _returnvalue (Transformation|GetActorTransform MyTarget))
  ;[0 0 0] Physical Surface
  (Networking|CanExecuteCosmeticEvents
    (:True
      (Utilities|FlowControl|Sequence
        (:then_0
          (bind (_outhit _returnvalue_2) (Collision|LineTraceForObjects 0 (+ (.location _returnvalue) "0.000000,0.000000,150.000000") (- (.location _returnvalue) "0.000000,0.000000,300.000000") (Variables|Default|GetObject%20Types)))
          (if _returnvalue_2
            (bind _returnvalue_3 (Effects|Components|ParticleSystem|SpawnFXAtLocation :FXSystemAsset "/NitroGauntletsGameplay/Assets/FX/NS_NitroGauntlets_Physical_Surface.NS_NitroGauntlets_Physical_Surface" :Location (.Location (Collision|BreakHitResult _outhit)) :Scale "1.000000,1.000000,1.000000" :bAutoDestroy true :PoolingMethod "AutoRelease" :bAutoActivate true))
            (if (.ReturnValue (Utilities|Map|Find (Variables|Default|GetPhysical%20mat%20Sand) (.PhysMat (Collision|BreakHitResult _outhit))))
              (Effects|Components|ParticleSystem|SetBoolParameter _returnvalue_3 "IsSand" true)
              (Effects|Components|ParticleSystem|SetBoolParameter _returnvalue_3 "IsGround" false)
              (elif (.ReturnValue (Utilities|Map|Find (Variables|Default|GetPhysical%20mat%20Snow) (.PhysMat (Collision|BreakHitResult _outhit))))
                (Effects|Components|ParticleSystem|SetBoolParameter _returnvalue_3 "IsSnow" true)
                (Effects|Components|ParticleSystem|SetBoolParameter _returnvalue_3 "IsGround" false)
                (elif (.ReturnValue (Utilities|Map|Find (Variables|Default|GetPhysical%20mat%20Grass) (.PhysMat (Collision|BreakHitResult _outhit))))
                  (Effects|Components|ParticleSystem|SetBoolParameter _returnvalue_3 "IsGrass" true)
                  (Effects|Components|ParticleSystem|SetBoolParameter _returnvalue_3 "IsGround" false))))
            (bind _returnvalue_4 (Development|GetConsoleVariableIntValue "r.PhysicsField.EnableField"))
            (if (== _returnvalue_4 1)
              (Game|SpawnActorfromClass 0 "/Game/Effects/Niagara/Blueprints/PhsyicsFields/FN_RadialForce.FN_RadialForce_C" _returnvalue "Undefined" "MultiplyWithRoot" 0 1.0 0.75 1e+03 1.7 0.0 0)
              (bind _asfort_player_pawn (Utilities|Casting|CastToFortPlayerPawn MyTarget))
              (if (Pawn|IsLocallyViewed _asfort_player_pawn)
                (Effects|Components|ParticleSystem|SetBoolParameter ParticleComponents "LocalView" true))
              (return))))
        (:then_1
          (Fortnite|ShowSoundIndicatorLocalforTeam :Location (Transformation|GetActorLocation MyTarget) :MaxAudibleDistance 3e+03 :IndicatorType "Gunshot" :Team (Team|GetActorTeam MyTarget) :Affiliations (Utilities|Array|MakeArray "Neutral" "Hostile") :Instigator MyTarget :Tint "(R=0.000000,G=0.000000,B=0.000000,A=0.000000)" :AttenuationMaxRangeMultiplier 1.0 :AttenuationIntensityMultiplier 1.0))))))

(fn UpdateFXLocation
  (params (FXComp :type Niagara%20Particle%20System%20Component%20Object%20Reference)))
