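; Burst gameplay cue for Nitro Gauntlets hits on vehicles, choosing the hit sound that matches the attack.
; On burst, it switches on the first aggregated source tag to map AttackA, AttackB or AttackC to an index.
; That index picks an entry from the vehicle hit sounds array, which is set on the first spawned audio component.
; A leftover-style Print String outputs the chosen index.
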(fn OnBurst(EitherVFX-preferthisversion)
  (params (MyTarget :type Actor%20Object%20Reference) (Parameters :type Gameplay%20Cue%20Parameters%20Structure%20%28by%20ref%29) (ParticleComponents :type Array%20of%20FXSystem%20Component%20Object%20References) (AudioComponents :type Array%20of%20Audio%20Component%20Object%20References) (BurstCameraShakeInstance :type Camera%20Shake%20Base%20Object%20Reference) (BurstDecalInstance :type Decal%20Actor%20Object%20Reference))
  (bind (_normalizedmagnitude _rawmagnitude _effectcontext _matchedtagname _originaltag _aggregatedsourcetags) (Ability|GameplayCue|BreakGameplayCueParameters Parameters))
  (bind _int (LocalVariables|GetInt))
  (switch Utilities|FlowControl|Switch|SwitchonGameplayTag (Utilities|Array|Get(acopy) (GameplayTags|BreakGameplayTagContainer _aggregatedsourcetags))
    (:Event.Moonflax.NitroGauntlets.AttackA
      (LocalVariables|SetInt 0))
    (:Event.Moonflax.NitroGauntlets.AttackB
      (LocalVariables|SetInt 1))
    (:Event.Moonflax.NitroGauntlets.AttackC
      (LocalVariables|SetInt 2)))
  (Audio|Components|Audio|SetSound (Utilities|Array|Get(acopy) AudioComponents) (Utilities|Array|Get(acopy) (Variables|Default|GetVehicleHitSounds) _int))
  (Development|PrintString 0 (Utilities|String|ToString(Integer) _int))
  (return))
